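# Base surface material for the petrol station pump prop, built on the layered MF_GigaMix_2023 function with material attributes.
# The layered result passes through a World Position Offset override carrying a documentation-link parameter, then destructible damage and Curie fallback effect functions.
# Normals are then converted to world space for possible RVT use, as tangent-space normals are off.
# A mobile shading-path branch substitutes constant specular and roughness before the Creative move-tool function.
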Material /Game/Environments/Helios/Props/Commerce/Commerce_PetrolStationPump_A/Materials/P_FN_Base_PetrolStationPump
graph {
 properties {
  bAutomaticallySetUsageInEditor=false
  bTangentSpaceNormal=false
  bUseMaterialAttributes=true
  bUsedWithInstancedStaticMeshes=true
  bUsedWithMeshDeformer=true
  bUsedWithNanite=true
  bUsedWithSkeletalMesh=true
  bUsedWithSplineMeshes=true
  ditheredLODTransition=true
  pixelDepthOffsetMode=PDOM_Legacy
  shadingModel=MSM_FromMaterialExpression
  usedShadingModels=Default Lit | Subsurface | Clear Coat | Two Sided Foliage
 }
 nodes [
  n1 = MaterialFunctionCall() {materialFunction=/Game/Global/Base4Layers_2022/MaterialFunctions/MF_GigaMix_2023}
  MaterialExpressionReroute_70 = Reroute(None=@n1.Result)
  n2 = ScalarParameter() {
   desc=Documentation link
   group="0 (Expand)  Documentation (RMB - Copy display name) and Debug view"
   parameterName=https://confluence.it.epicgames.com/display/FORT/Fortnite+Base+Material+2022
  }
  n3 = SetMaterialAttributes!(MaterialAttributes=@MaterialExpressionReroute_70.MaterialAttributes, World Position Offset=Multiply(A=@n2)) {attributeSetTypes [#1]}
  MaterialExpressionReroute_71 = Reroute(None=@n3)
  MaterialExpressionReroute_69 = Reroute(None=@MaterialExpressionReroute_71.Output)
  n4 = MaterialFunctionCall!(
   In=MaterialFunctionCall(Material_In=@MaterialExpressionReroute_69.MaterialAttributes) {materialFunction=&MF_FN_DestructibleDamage}.Material_Out
   UseSphericalMappingOverride
  ) {materialFunction=/Game/Packages/Fortress_SharedMaterials/Base_Material_Functions/MF_CurieFallbackEffect}
  MaterialExpressionReroute_68 = Reroute(None=@n4.Result)
  Composite!(None=@MaterialExpressionReroute_68.Result) {desc=Transform to World in case we will use RVT after this point, subgraphName=Tangent to World}
  Composite!(MaterialAttributes=@n1.Result) {subgraphName=Documentation Link}
  Composite(MaterialAttributes=@MaterialExpressionReroute_71.Output) {desc=Game Shared WPO and Fire Effect, subgraphName=WPO And Shared Functions}
  MaterialExpressionReroute_3 = Reroute(None=@MaterialExpressionReroute_68.Result)
  n5 = GetMaterialAttributes(None=@MaterialExpressionReroute_3.Input) {attributeGetTypes [#2]}
  n6 = SetMaterialAttributes!(MaterialAttributes=@n5.MaterialAttributes, Normal=Transform(None=@n5.Normal)) {attributeSetTypes [#2]}
  MaterialExpressionReroute_2 = Reroute(None=@n6)
  Composite!(InputPin=@MaterialExpressionReroute_2.Output) {subgraphName=Mobile Opt}
  MaterialExpressionReroute_7 = Reroute(None=@MaterialExpressionReroute_2.Output)
  n7 = GetMaterialAttributes(None=@MaterialExpressionReroute_7.InputPin) {attributeGetTypes [#3, #2]}
  n8 = SetMaterialAttributes!(
   MaterialAttributes=@n7.MaterialAttributes
   Specular=Constant() {r=0.5}
   Roughness=Constant() {r=0.8}
   Base Color=@n7.MaterialAttributes
   Normal=@n7.MaterialAttributes
  ) {attributeSetTypes [
    #4
    #5
    #3
    #2
   ]}
  n9 = ShadingPathSwitch(Default=@MaterialExpressionReroute_7.InputPin, Mobile=@n8) {desc=Migrated from FeatureLevelSwitch}
  MaterialExpressionReroute_6 = Reroute(None=@n9)
  MaterialFunctionCall(MA In=@MaterialExpressionReroute_6.Output) {materialFunction=&MF_CreativeMoveTool}
  PinBase() {reroutePins [Input {expression=&self:MaterialExpressionReroute_3}]}
  PinBase!(Output=@n6) {reroutePins [Output {expression=&self:MaterialExpressionReroute_2}]}
  PinBase!() {reroutePins [MaterialAttributes {expression=&self:MaterialExpressionReroute_70}]}
  PinBase!(Output=@n3) {reroutePins [Output {expression=&self:MaterialExpressionReroute_71}]}
  PinBase!() {reroutePins [MaterialAttributes {expression=&self:MaterialExpressionReroute_69}]}
  PinBase!(Result=@n4.Result) {reroutePins [Result {expression=&self:MaterialExpressionReroute_68}]}
  PinBase!() {reroutePins [InputPin {expression=&self:MaterialExpressionReroute_7}]}
  PinBase!(Output=@n9) {reroutePins [Output {expression=&self:MaterialExpressionReroute_6}]}
 ]
}
